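# Unlit translucent lightning bolt material for particle effects, flagged for mesh particles and Niagara ribbons.
# A swizzled plasma texture pans by dynamic parameters, blending edge and core colors tinted by particle color and remapped to HDR.
# Its alpha branch combines texture channels, particle alpha and a U-edge fade.
# Use_lightning_link blends between a flat bright color with vertex alpha and this animated bolt.
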Material /Game/Marketing_Screenshots/E32015/Materials/M_LightningBolt_01
graph {
 properties {
  bUsedWithMeshParticles=true
  bUsedWithNiagaraRibbons=true
  blendMode=BLEND_Translucent
  pixelDepthOffsetMode=PDOM_Legacy
  shadingModel=MSM_Unlit
  usedShadingModels=Unlit
 }
 nodes [
  n1 = ComponentMask(None=TextureCoordinate()) {r=true}
  n2 = OneMinus(None=@n1)
  n3 = Multiply(A=Power(Base=Multiply(A=@n1, B=@n2))) {constB=10}
  n4 = TextureCoordinate() {uTiling=0.5}
  n5 = DynamicParameter() {defaultValue=<rgba 1 0 1 20>, paramNames [Time Multiplier, UV Offset V, U Scale, Random2]}
  n6 = AppendVector(A=@n5.U Scale, B=Constant() {r=1})
  n7 = Add(
   A=MaterialFunctionCall(XY=Multiply(A=@n4, B=@n6)) {materialFunction=&Swizzle}.YX
   B=AppendVector(A=Constant(), B=@n5.UV Offset V)
  )
  n8 = TextureSampleParameter2D(UVs=Panner(Coordinate=Multiply(A=@n7), Time=@n5.Time Multiplier) {speedY=-1}) {parameterName=Texture, texture=/Game/Effects/Fort_Effects/Textures/Tile/T_Lightning_Plasma_Soft}
  n9 = Add(A=@n8.R, B=Multiply(A=@n8.B) {constB=0.2})
  n10 = ParticleColor()
  LinearInterpolate(A=VertexColor().A, B=Multiply(A=Clamp(None=@n3), B=Multiply(A=@n9, B=@n10.A)), Alpha=ScalarParameter() {parameterName=Use_lightning_link})
  n11 = VectorParameter() {parameterName=v, defaultValue=<rgba 7.8865 7.75 12 0>}
  n12 = Constant() {r=2.5}
  n13 = Multiply(
   A=@n10.RGB
   B=LinearInterpolate(
    A=VectorParameter() {parameterName=EdgeColor, defaultValue=<rgba 0.2226 0.2425 0.5 0>}.RGB
    B=VectorParameter() {parameterName=Color, defaultValue=<rgba 0.6214 0.9501 1 0>}.RGB
    Alpha=@n8.G
   )
  )
  n14 = Clamp(None=@n13)
  n15 = DynamicParameter() {defaultValue=<rgba 1 0 1 20>, paramNames [Time Multiplier, UV Offset V, U Scale, Random2]}
  MaterialFunctionCall(Emissive=LinearInterpolate(
    A=Multiply(A=@n11.RGB, B=@n12)
    B=MaterialFunctionCall(LDR Input=@n14, HDR Tint=@n13, HDR Intensity Multiplier=@n15.Random2) {materialFunction=&DeriveHDRfromLDR}.RemappedRange
    Alpha=ScalarParameter() {parameterName=Use_lightning_link}
   )) {materialFunction=&MF_FX_EmissiveAdjustments}
  MaterialFunctionCall() {materialFunction=&DebugTimeSine}
  Multiply(A=Time(), B=Multiply(A=ScalarParameter() {defaultValue=1, parameterName=PanSpeed}, B=@n5.Time Multiplier))
  ScalarParameter() {defaultValue=20, parameterName=HDRIntensity}
 ]
}
